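# Translucent unlit Niagara material for the BroadPoodle guitar pickaxe impact effect, drawing a stylized radial burst of inky smoke.
# Radial-UV smoke noise, panned by dynamic PanX and sampled with base-UV derivatives, drives a three-color blend scaled by time of day and particle color.
# Alpha combines radial and dynamic Erosion, rounds it to hard edges, then applies a radial mask, second randomly offset noise breakup, depth fade and particle alpha.
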Material /SparksCosmetics/Instrument/Guitar/BroadPoodle/Effects/M_Pickaxe_BroadPoodle_Impact
graph {
 properties {
  bUsedWithNiagaraMeshParticles=true
  bUsedWithNiagaraRibbons=true
  bUsedWithNiagaraSprites=true
  blendMode=BLEND_Translucent
  pixelDepthOffsetMode=PDOM_Legacy
  shadingModel=MSM_Unlit
  usedShadingModels=Unlit
 }
 nodes [
  n1 = MaterialFunctionCall!(Center, Maintain Aspect Ratio, Tiling, Offset, In Tex Coords=TextureCoordinate()) {materialFunction=/Game/Effects/Niagara/MasterMaterials/Functions/Mojo_MF/MF_FX_Uniform_Radial_UV}
  n2 = AppendVector(A=ScalarParameter() {defaultValue=0.35, parameterName=TileX}, B=ScalarParameter() {defaultValue=1.25, parameterName=TileY})
  n3 = DynamicParameter() {defaultValue=<rgba 1 0 1 1>, paramNames [PanX,Erosion,Param3,Param4]}
  n4 = OneMinus(None=@n3.PanX)
  n5 = Constant()
  n6 = TextureCoordinate()
  n7 = TextureSampleParameter2D!(
   UVs=Add(A=Multiply(A=@n1.Out UV, B=@n2), B=AppendVector(A=@n4, B=@n5))
   "DDX(UVs)"=DDX(Value=@n6)
   "DDY(UVs)"=DDY(Value=@n6)
   Apply View MipBias
  ) {mipValueMode=TMVM_Derivative, parameterName=Main Tex, texture=&T_Inky_Smoke_Tile}
  n8 = Multiply(A=@n7.RGB)
  n9 = ScalarParameter() {defaultValue=0.5, parameterName=ColorPower}
  n10 = Power(Base=Multiply(A=ComponentMask(None=@n1.Out UV) {r=true}) {constB=0.5}, Exp=Constant() {r=2})
  n11 = Multiply(A=MaterialFunctionCall!(
    Gradient=Power(Base=@n8, Exp=@n9) {constExponent=0.02}
    Erosion Amount=@n10
    Softness=ScalarParameter() {defaultValue=5, parameterName=ColorErosionSoftness}
   ) {materialFunction=/BRCosmetics/Athena/Effects/Particles/Skydiving/NS_Contrail_DarkAndLight/Functions/MF_Erosion}.Default Range) {constB=5}
  Add(A=@n11, B=@n7.R)
  AppendVector(A=ScalarParameter() {parameterName=SpeedY}, B=@n3.PanX)
  Constant() {r=1}
  Constant() {r=2}
  MaterialFunctionCall(XY=@n6) {materialFunction=&Swizzle}
  n12 = MaterialFunctionCall(Radius=Constant() {r=0.25}) {materialFunction=&RadialGradientExponential}
  n13 = LinearInterpolate(
   A=LinearInterpolate(
    A=VectorParameter() {parameterName=Color Tertiary, defaultValue=<rgba 0.0344 0.1268 0.2135 1>}.RGB
    B=VectorParameter() {parameterName=Color Secondary, defaultValue=<rgba 0.0489 0.1589 0.1615 1>}.RGB
    Alpha=@n12.RadialGradientExponential
   )
   B=VectorParameter() {parameterName=Color Primary, defaultValue=<rgba 1 0 0.2846 1>}.RGB
   Alpha=@n11
  )
  n14 = Constant() {r=0.85}
  n15 = MaterialFunctionCall(Value_Morning=@n14, Value_Day=Constant() {r=1}, Value_Evening=@n14, Value_Night=Constant() {r=0.7}) {materialFunction=&Scalar_Time_of_Day}
  MaterialFunctionCall(Emissive=Multiply(A=Multiply(A=@n13, B=@n15.Result), B=ParticleColor().RGB)) {materialFunction=&MF_FX_EmissiveAdjustments}
  n16 = Power(Base=@n7.R, Exp=ScalarParameter() {defaultValue=0.5, group="🔍Alpha", parameterName=AlphaPower})
  n17 = ScalarParameter() {defaultValue=0.5, group="🔍Alpha", parameterName=lll}
  n18 = ScalarParameter() {defaultValue=1, group="🔍Alpha", parameterName=s}
  n19 = ComponentMask(None=@n1.Out UV) {r=true}
  n20 = MaterialFunctionCall(
   Gradient=@n16
   Erosion Amount=OneMinus(None=SmoothStep(Min=@n17, Max=@n18, Value=@n19))
   Softness=ScalarParameter() {defaultValue=1, group="🔍Alpha", parameterName=AlphaErosionSoftness}
  ) {materialFunction=&MF_Erosion}
  n21 = MaterialFunctionCall(
   Gradient=@n7.R
   Erosion Amount=DynamicParameter() {defaultValue=<rgba 1 0 1 1>, paramNames [PanX,Erosion,Param3,Param4]}.Erosion
   Softness=ScalarParameter() {defaultValue=0.5, group="🔍Alpha", parameterName=OverallErosion_Softness}
  ) {materialFunction=&MF_Erosion}
  n22 = Subtract(A=Multiply(A=Saturate(None=@n20.Default Range), B=@n21.Increased Range), B=OneMinus(None=@n21.Increased Range))
  n23 = LinearInterpolate(
   A=Multiply(A=Saturate(None=@n22), B=ScalarParameter() {defaultValue=1, group="🔍Alpha", parameterName=CoreAlphaIntensity}) {constB=0.5}
   Alpha=NamedRerouteUsage()
  ) {constA=0.2}
  n24 = ComponentMask(None=TextureCoordinate()) {r=true}
  Multiply(
   A=@n23
   B=Power(Base=OneMinus(None=@n24), Exp=ScalarParameter() {defaultValue=15, group="🔍Alpha", parameterName=Alpha_GradientPower}) {constExponent=8}
  )
  Multiply(A=@n10) {constB=2}
  n25 = ComponentMask(None=@n1.Out UV) {r=true}
  Multiply(A=OneMinus(None=@n25))
  Multiply(A=@n12.RadialGradientExponential) {constB=7}
  n26 = OneMinus(None=@n1.Out UV)
  Multiply(A=@n11, B=Power(Base=ComponentMask(None=@n26) {r=true}, Exp=Constant() {r=2}))
  Multiply(A=@n7.R, B=@n23)
  NamedRerouteDeclaration(None=@n11) {name=SmokeyNoise, nodeColor=<rgba 0 1 0.5313 1>}
  OneMinus(None=@n19)
  n27 = Saturate(None=Power(Base=ComponentMask(None=@n1.Out UV) {r=true}, Exp=Constant() {r=2}))
  OneMinus(None=Saturate(None=LinearInterpolate(Alpha=@n27) {constA=1, constB=-0.05}))
  ParticleRandom()
  Power(Base=@n25) {constExponent=8}
  n28 = Round(None=@n23)
  n29 = Multiply(A=MaterialFunctionCall() {materialFunction=&RadialGradientExponential}.RadialGradientExponential) {constB=3}
  n30 = MaterialFunctionCall(Tiling=Constant2Vector() {g=1.15, r=0.8}, In Tex Coords=TextureCoordinate()) {materialFunction=&MF_FX_Uniform_Radial_UV}
  n31 = AppendVector(A=Constant() {desc=tilex, r=0.35}, B=Constant() {desc=tiley, r=1.25})
  n32 = OneMinus(None=DynamicParameter() {defaultValue=<rgba 1 0 1 1>, paramNames [PanX,Erosion,Param3,Param4]}.PanX)
  n33 = ParticleRandom()
  n34 = TextureCoordinate()
  n35 = TextureSampleParameter2D(UVs=Add(A=Multiply(A=@n30.Out UV, B=@n31), B=AppendVector(A=@n32, B=@n33)), "DDX(UVs)"=DDX(Value=@n34), "DDY(UVs)"=DDY(Value=@n34)) {mipValueMode=TMVM_Derivative, parameterName=Main Tex, texture=&T_Inky_Smoke_Tile}
  n36 = Multiply(A=Power(Base=@n35.RGB)) {constB=25}
  n37 = DepthFade(
   Opacity=Multiply(A=Multiply(A=@n28, B=@n29), B=Saturate(None=@n36)) {desc=Breakup Mask}
   FadeDistance=ScalarParameter() {defaultValue=2, group="🔍Alpha", parameterName=Alpha_DepthFade}
  )
  n38 = ParticleColor()
  Reroute(None=Multiply(A=Multiply(A=@n37, B=@n38.A), B=ScalarParameter() {defaultValue=1, parameterName=Alpha_GlobalOpacity}))
  Saturate(None=LinearInterpolate(A=Constant() {r=1}, B=Constant(), Alpha=@n27) {constA=1})
  Saturate(None=@n11)
  Saturate(None=Multiply(A=@n16, B=@n7.R))
  ScalarParameter() {defaultValue=3, parameterName=SpeedX}
  TextureCoordinate()
  TextureCoordinate()
  TextureCoordinate()
  TextureCoordinate()
 ]
}
