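# Translucent unlit smoky trail material for the BroadPoodle guitar pickaxe, usable on Niagara ribbons and mesh particles.
# A tiled inky smoke texture, panned by dynamic parameter PanX, drives MF_Erosion masks for opacity and color.
# Alpha combines lengthwise erosion, particle-driven Erosion and edge fades, modulated by the color's SmokeyNoise mask and particle alpha.
# Emissive color blends three tint parameters by UV gradient and eroded noise, then scales by time of day and particle color.
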Material /SparksCosmetics/Instrument/Guitar/BroadPoodle/Effects/M_Pickaxe_BroadPoodle_Trail
graph {
 properties {
  bUsedWithNiagaraMeshParticles=true
  bUsedWithNiagaraRibbons=true
  blendMode=BLEND_Translucent
  pixelDepthOffsetMode=PDOM_Legacy
  shadingModel=MSM_Unlit
  usedShadingModels=Unlit
 }
 nodes [
  n1 = DynamicParameter() {defaultValue=<rgba 1 0 1 1>, paramNames [PanX,Erosion,Param3,Param4]}
  AppendVector(A=ScalarParameter() {parameterName=SpeedY}, B=@n1.PanX)
  Constant() {r=1}
  n2 = MaterialFunctionCall(XY=TextureCoordinate()) {materialFunction=&Swizzle}
  n3 = AppendVector(A=ScalarParameter() {defaultValue=1.25, parameterName=TileX}, B=ScalarParameter() {defaultValue=1.75, parameterName=TileY})
  n4 = Constant()
  n5 = TextureSampleParameter2D(UVs=Add(A=Multiply(A=@n2.YX, B=@n3), B=AppendVector(A=@n4, B=@n1.PanX))) {parameterName=Main Tex, texture=&T_Inky_Smoke_Tile}
  n6 = ScalarParameter() {defaultValue=0.5, group="🔍Alpha", parameterName=AlphaPower}
  n7 = ScalarParameter() {defaultValue=0.5, group="🔍Alpha", parameterName=lll}
  n8 = ScalarParameter() {defaultValue=1, group="🔍Alpha", parameterName=s}
  n9 = ComponentMask(None=TextureCoordinate()) {r=true}
  n10 = TextureCoordinate()
  n11 = Multiply(
   A=MaterialFunctionCall!(
    Gradient=Power(Base=@n5.R, Exp=@n6)
    Erosion Amount=SmoothStep(Min=@n7, Max=@n8, Value=@n9)
    Softness=ScalarParameter() {defaultValue=1, group="🔍Alpha", parameterName=AlphaErosionSoftness}
   ) {materialFunction=/BRCosmetics/Athena/Effects/Particles/Skydiving/NS_Contrail_DarkAndLight/Functions/MF_Erosion}.Default Range
   B=SmoothStep(
    Min=ScalarParameter() {group="🔍Alpha", parameterName=Param_4}
    Max=ScalarParameter() {defaultValue=0.02, group="🔍Alpha", parameterName=Param_5}
    Value=ComponentMask(None=@n10) {r=true}
   )
  )
  n12 = ComponentMask(None=TextureCoordinate()) {r=true}
  n13 = MaterialFunctionCall(
   Gradient=Multiply(A=@n5.R, B=OneMinus(None=@n12))
   Erosion Amount=DynamicParameter() {defaultValue=<rgba 1 0 1 1>, paramNames [PanX,Erosion,Param3,Param4]}.Erosion
   Softness=ScalarParameter() {defaultValue=0.35, group="🔍Alpha", parameterName=OverallErosion_Softness}
  ) {materialFunction=&MF_Erosion}
  n14 = Subtract(A=Multiply(A=Saturate(None=@n11), B=@n13.Increased Range), B=OneMinus(None=@n13.Increased Range))
  n15 = ScalarParameter() {defaultValue=0.1, group="🔍Alpha", parameterName=VerticalEdgeFadeFallOff}
  n16 = Subtract(A=ComponentMask(None=TextureCoordinate()) {g=true}) {constB=0.5}
  n17 = Power(Base=OneMinus(None=Abs(None=@n16)), Exp=ScalarParameter() {defaultValue=1.5, group="🔍Alpha", parameterName=VerticalEdgeFade}) {constExponent=8}
  n18 = Multiply(
   A=Multiply(A=Saturate(None=@n14), B=SmoothStep(Min=@n15, Value=@n17))
   B=ScalarParameter() {defaultValue=10, group="🔍Alpha", parameterName=VerticalEdgeFadeIntensity}
  )
  n19 = LinearInterpolate(
   A=Multiply(A=@n18, B=ScalarParameter() {defaultValue=1, group="🔍Alpha", parameterName=CoreAlphaIntensity}) {constB=0.5}
   B=@n18
   Alpha=NamedRerouteUsage()
  ) {constA=0.2}
  DepthFade(Opacity=Round(None=@n19), FadeDistance=ScalarParameter() {defaultValue=2, group="🔍Alpha", parameterName=Alpha_DepthFade})
  n20 = TextureCoordinate()
  n21 = Saturate(None=Power(Base=ComponentMask(None=@n20) {r=true}, Exp=Constant() {r=2}))
  n22 = LinearInterpolate(Alpha=@n21) {constA=1, constB=-5}
  n23 = ComponentMask(None=OneMinus(None=TextureCoordinate())) {r=true}
  n24 = Constant() {r=2}
  n25 = Power(Base=@n5.RGB, Exp=ScalarParameter() {defaultValue=1, parameterName=ColorPower}) {constExponent=0.02}
  n26 = Power(Base=ComponentMask(None=TextureCoordinate()) {r=true}, Exp=ScalarParameter() {defaultValue=4, parameterName=Primary Color Erosion Amount})
  n27 = ScalarParameter() {defaultValue=0.35, parameterName=ColorErosionSoftness}
  n28 = Saturate(None=Multiply(A=MaterialFunctionCall(Gradient=@n25, Erosion Amount=@n26, Softness=@n27) {materialFunction=&MF_Erosion}.Default Range) {constB=2})
  n29 = LinearInterpolate(
   A=LinearInterpolate(
    A=VectorParameter() {parameterName=Color Tertiary, defaultValue=<rgba 0.0244 0.0952 0.1562 1>}.RGB
    B=VectorParameter() {parameterName=Color Secondary, defaultValue=<rgba 0.0458 0.1486 0.151 1>}.RGB
    Alpha=Saturate(None=@n22)
   )
   B=VectorParameter() {parameterName=Color Primary, defaultValue=<rgba 0.4427 0.1735 0.2521 1>}.RGB
   Alpha=Multiply(A=@n28, B=Power(Base=@n23, Exp=@n24))
  )
  n30 = Constant() {r=0.85}
  n31 = MaterialFunctionCall(Value_Morning=@n30, Value_Day=Constant() {r=1}, Value_Evening=@n30, Value_Night=Constant() {r=0.5}) {materialFunction=&Scalar_Time_of_Day}
  MaterialFunctionCall(Emissive=Multiply(A=Multiply(A=@n29, B=@n31.Result), B=ParticleColor().RGB)) {materialFunction=&MF_FX_EmissiveAdjustments}
  n32 = ComponentMask(None=TextureCoordinate()) {r=true}
  Multiply(
   A=@n19
   B=Power(Base=OneMinus(None=@n32), Exp=ScalarParameter() {defaultValue=15, group="🔍Alpha", parameterName=Alpha_GradientPower}) {constExponent=8}
  )
  NamedRerouteDeclaration(None=@n28) {name=SmokeyNoise, nodeColor=<rgba 0 1 0.5313 1>}
  OneMinus(None=@n9)
  n33 = Constant() {r=1}
  n34 = Constant() {r=-0.7}
  n35 = Multiply(A=Saturate(None=LinearInterpolate(A=@n33, B=@n34, Alpha=@n21) {constA=1}), B=@n19)
  n36 = ParticleColor()
  Reroute(None=Multiply(A=Multiply(A=@n35, B=@n36.A), B=ScalarParameter() {defaultValue=1, parameterName=Alpha_GlobalOpacity}))
  ScalarParameter() {defaultValue=3, parameterName=SpeedX}
 ]
}
